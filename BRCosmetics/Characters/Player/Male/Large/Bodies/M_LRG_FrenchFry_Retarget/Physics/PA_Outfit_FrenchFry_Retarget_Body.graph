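# Physics asset for the retargeted FrenchFry outfit body on the large male player skeleton.
# One capsule per bone covers pelvis, upper spine, head, arms, hands with pinky metacarpals, legs and feet, including several twist bones.
# Constraints form a tree rooted at the pelvis with uniform swing and twist limit angles.
# The main thigh bones have no bodies, so calves and thigh twist bodies constrain directly to the pelvis.
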PhysicsAsset /BRCosmetics/Characters/Player/Male/Large/Bodies/M_LRG_FrenchFry_Retarget/Physics/PA_Outfit_FrenchFry_Retarget_Body
bodies {
 calf_l [{shapeType=Capsule, length=19.6631, radius=14.4464, center=<-9.738 -1.7383 1.0501>, rotation=<rot 0 148.0972 -85.0617>}]
 calf_r [{shapeType=Capsule, length=19.6631, radius=14.4463, center=<9.7381 1.7383 -1.0501>, rotation=<rot 0 148.0972 -85.0614>}]
 calf_twist_02_l [{shapeType=Capsule, length=17.6104, radius=14.8661, center=<-5.2197 1.1238 0.3327>, rotation=<rot 0 -175.2363 -69.1052>}]
 calf_twist_02_r [{shapeType=Capsule, length=17.6104, radius=14.8661, center=<5.2197 -1.1238 -0.3327>, rotation=<rot 0 -175.2363 -69.1052>}]
 clavicle_l [{shapeType=Capsule, length=17.8338, radius=9.4978, center=<15.1062 -2.8875 0.3831>, rotation=<rot 0 1.8725 -73.6339>}]
 clavicle_r [{shapeType=Capsule, length=17.834, radius=9.4978, center=<-15.1058 2.8877 -0.3832>, rotation=<rot 0 1.8729 -73.6338>}]
 foot_l [{shapeType=Capsule, length=22.5703, radius=14.4332, center=<-2.7873 6.7136 1.7334>, rotation=<rot 0 177.949 -79.92>}]
 foot_r [{shapeType=Capsule, length=22.5703, radius=14.4332, center=<2.7874 -6.7136 -1.7334>, rotation=<rot 0 177.949 -79.92>}]
 hand_l [{shapeType=Capsule, length=12.3231, radius=8.1031, center=<5.591 -3.1658 -0.9845>, rotation=<rot 0 103.4133 -70.3946>}]
 hand_r [{shapeType=Capsule, length=12.3269, radius=8.0936, center=<-5.5875 3.1667 0.9873>, rotation=<rot 0 103.2982 -70.3245>}]
 head [{shapeType=Capsule, length=20.1373, radius=8.7873, center=<-3.0071 -16.1787 0.0016>, rotation=<rot 0.0201 -90 -0.0016>}]
 lowerarm_l [{shapeType=Capsule, length=15.2463, radius=12.8417, center=<7.9157 0.3127 -3.4353>, rotation=<rot -0.348 -90 -71.1635>}]
 lowerarm_r [{shapeType=Capsule, length=15.2462, radius=12.8417, center=<-7.9157 -0.3127 3.4353>, rotation=<rot -0.3478 -90 -71.1635>}]
 pelvis [{shapeType=Capsule, length=37.6998, radius=36.9423, center=<4.8668 -0.4959 -0.1906>, rotation=<rot 8.469 99.2022 -90>}]
 pinky_metacarpal_l [{shapeType=Capsule, length=11.3343, radius=6.8067, center=<3.2423 -2.2734 1.8744>, rotation=<rot 0 -111.0237 -64.2279>}]
 pinky_metacarpal_r [{shapeType=Capsule, length=11.3343, radius=6.8067, center=<-3.2423 2.2734 -1.8744>, rotation=<rot 0 -111.0237 -64.2279>}]
 spine_04 [{shapeType=Capsule, length=23.4921, radius=13.861, center=<9.6137 -17.3607 -0.0021>, rotation=<rot 0.0108 -90 0.0175>}]
 spine_05 [{shapeType=Capsule, length=26.4476, radius=12.0216, center=<-9.3943 18.5174 0.0161>, rotation=<rot 0.3788 -90 -0.521>}]
 thigh_twist_01_l [{shapeType=Capsule, length=20.3449, radius=11.5098, center=<0.5192 -1.26 -7.2314>, rotation=<rot 0 2.7222 -83.8019>}]
 thigh_twist_01_r [{shapeType=Capsule, length=20.3447, radius=11.5099, center=<-0.5192 1.2596 7.2313>, rotation=<rot 0 2.722 -83.802>}]
 thigh_twist_02_l [{shapeType=Capsule, length=17.6332, radius=15.0291, center=<-1.4273 0.0003 1.1993>, rotation=<rot 0 173.4192 -79.488>}]
 thigh_twist_02_r [{shapeType=Capsule, length=17.7872, radius=15.4708, center=<1.2485 -0.0444 -1.6659>, rotation=<rot 0 172.6015 -74.7029>}]
 upperarm_l [{shapeType=Capsule, length=15.0007, radius=11.3121, center=<20.3735 0.1654 -0.6802>, rotation=<rot 0 8.5259 -82.1168>}]
 upperarm_r [{shapeType=Capsule, length=15.0187, radius=11.3788, center=<-20.352 -0.1608 0.715>, rotation=<rot 0 8.6932 -81.5287>}]
 upperarm_twist_01_l [{shapeType=Capsule, length=18.5002, radius=9.4844, center=<-2.5062 -2.2952 0.4074>, rotation=<rot 0 -167.4287 -85.6209>}]
 upperarm_twist_01_r [{shapeType=Capsule, length=18.5002, radius=9.4844, center=<2.5062 2.2952 -0.4074>, rotation=<rot 0 -167.4287 -85.6209>}]
}
constraints [
 {bone1Name=spine_04, bone2Name=pelvis, swing1LimitDegrees=45, swing2LimitDegrees=45, twistLimitDegrees=45}
 {bone1Name=spine_05, bone2Name=spine_04, swing1LimitDegrees=45, swing2LimitDegrees=45, twistLimitDegrees=45}
 {bone1Name=head, bone2Name=spine_05, swing1LimitDegrees=45, swing2LimitDegrees=45, twistLimitDegrees=45}
 {bone1Name=clavicle_l, bone2Name=spine_05, swing1LimitDegrees=45, swing2LimitDegrees=45, twistLimitDegrees=45}
 {bone1Name=upperarm_l, bone2Name=clavicle_l, swing1LimitDegrees=45, swing2LimitDegrees=45, twistLimitDegrees=45}
 {bone1Name=upperarm_twist_01_l, bone2Name=upperarm_l, swing1LimitDegrees=45, swing2LimitDegrees=45, twistLimitDegrees=45}
 {bone1Name=lowerarm_l, bone2Name=upperarm_l, swing1LimitDegrees=45, swing2LimitDegrees=45, twistLimitDegrees=45}
 {bone1Name=hand_l, bone2Name=lowerarm_l, swing1LimitDegrees=45, swing2LimitDegrees=45, twistLimitDegrees=45}
 {bone1Name=pinky_metacarpal_l, bone2Name=hand_l, swing1LimitDegrees=45, swing2LimitDegrees=45, twistLimitDegrees=45}
 {bone1Name=clavicle_r, bone2Name=spine_05, swing1LimitDegrees=45, swing2LimitDegrees=45, twistLimitDegrees=45}
 {bone1Name=upperarm_r, bone2Name=clavicle_r, swing1LimitDegrees=45, swing2LimitDegrees=45, twistLimitDegrees=45}
 {bone1Name=upperarm_twist_01_r, bone2Name=upperarm_r, swing1LimitDegrees=45, swing2LimitDegrees=45, twistLimitDegrees=45}
 {bone1Name=lowerarm_r, bone2Name=upperarm_r, swing1LimitDegrees=45, swing2LimitDegrees=45, twistLimitDegrees=45}
 {bone1Name=hand_r, bone2Name=lowerarm_r, swing1LimitDegrees=45, swing2LimitDegrees=45, twistLimitDegrees=45}
 {bone1Name=pinky_metacarpal_r, bone2Name=hand_r, swing1LimitDegrees=45, swing2LimitDegrees=45, twistLimitDegrees=45}
 {bone1Name=calf_l, bone2Name=pelvis, swing1LimitDegrees=45, swing2LimitDegrees=45, twistLimitDegrees=45}
 {bone1Name=foot_l, bone2Name=calf_l, swing1LimitDegrees=45, swing2LimitDegrees=45, twistLimitDegrees=45}
 {bone1Name=calf_twist_02_l, bone2Name=calf_l, swing1LimitDegrees=45, swing2LimitDegrees=45, twistLimitDegrees=45}
 {bone1Name=thigh_twist_01_l, bone2Name=pelvis, swing1LimitDegrees=45, swing2LimitDegrees=45, twistLimitDegrees=45}
 {bone1Name=thigh_twist_02_l, bone2Name=pelvis, swing1LimitDegrees=45, swing2LimitDegrees=45, twistLimitDegrees=45}
 {bone1Name=calf_r, bone2Name=pelvis, swing1LimitDegrees=45, swing2LimitDegrees=45, twistLimitDegrees=45}
 {bone1Name=foot_r, bone2Name=calf_r, swing1LimitDegrees=45, swing2LimitDegrees=45, twistLimitDegrees=45}
 {bone1Name=calf_twist_02_r, bone2Name=calf_r, swing1LimitDegrees=45, swing2LimitDegrees=45, twistLimitDegrees=45}
 {bone1Name=thigh_twist_01_r, bone2Name=pelvis, swing1LimitDegrees=45, swing2LimitDegrees=45, twistLimitDegrees=45}
 {bone1Name=thigh_twist_02_r, bone2Name=pelvis, swing1LimitDegrees=45, swing2LimitDegrees=45, twistLimitDegrees=45}
]
