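# Physics asset for the head of the FrenchFry Retarget outfit, a large male player cosmetic body.
# It holds two capsule bodies, one on the head bone and one on the spine_05 upper-spine bone.
# A single constraint joins the two bodies, limiting both swing axes and twist to the same moderate angle.
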PhysicsAsset /BRCosmetics/Characters/Player/Male/Large/Bodies/M_LRG_FrenchFry_Retarget/Physics/PA_Outfit_FrenchFry_Retarget_Head
bodies {
 head [{shapeType=Capsule, length=24.6505, radius=20.2496, center=<5.2923 2.5189 -0.0027>, rotation=<rot 0 -153.0806 -89.9597>}]
 spine_05 [{shapeType=Capsule, length=20.1476, radius=18.1591, center=<-0.1084 10.7699 0>, rotation=<rot 0.0001 -90 -0.0003>}]
}
constraints [{bone1Name=head, bone2Name=spine_05, swing1LimitDegrees=45, swing2LimitDegrees=45, twistLimitDegrees=45}]
