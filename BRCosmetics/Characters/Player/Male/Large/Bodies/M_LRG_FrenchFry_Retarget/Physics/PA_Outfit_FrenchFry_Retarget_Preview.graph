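# Preview physics asset for the retargeted FrenchFry outfit body on the large male player skeleton.
# Each major bone, several limb twist bones and both pinky metacarpals get one capsule, roughly mirrored left to right.
# Constraints chain from pelvis through upper spine to head and arms, all with uniform 45-degree swing and twist limit values.
# Main thigh bones have no bodies, so calves and thigh twist bodies constrain directly to the pelvis.
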PhysicsAsset /BRCosmetics/Characters/Player/Male/Large/Bodies/M_LRG_FrenchFry_Retarget/Physics/PA_Outfit_FrenchFry_Retarget_Preview
bodies {
 calf_l [{shapeType=Capsule, length=18.9264, radius=14.2536, center=<-8.5343 -1.8655 1.2079>, rotation=<rot 0 137.41 -86.9212>}]
 calf_r [{shapeType=Capsule, length=18.9264, radius=14.2535, center=<8.5342 1.8656 -1.2079>, rotation=<rot 0 137.4093 -86.9204>}]
 calf_twist_02_l [{shapeType=Capsule, length=16.6132, radius=14.9503, center=<-5.0627 2.321 0.2202>, rotation=<rot 0 -176.4947 -53.3789>}]
 calf_twist_02_r [{shapeType=Capsule, length=16.6131, radius=14.9503, center=<5.0622 -2.321 -0.2203>, rotation=<rot 0 -176.495 -53.3786>}]
 clavicle_l [{shapeType=Capsule, length=17.6522, radius=9.3965, center=<14.4259 -3.1124 0.6165>, rotation=<rot 0 1.583 -74.5595>}]
 clavicle_r [{shapeType=Capsule, length=17.6524, radius=9.3965, center=<-14.4259 3.1126 -0.6166>, rotation=<rot 0 1.5834 -74.5594>}]
 foot_l [{shapeType=Capsule, length=22.5512, radius=14.4277, center=<-3.353 6.698 1.7887>, rotation=<rot 0 177.3904 -79.0016>}]
 foot_r [{shapeType=Capsule, length=22.5512, radius=14.4277, center=<3.353 -6.698 -1.7887>, rotation=<rot 0 177.3904 -79.0016>}]
 hand_l [{shapeType=Capsule, length=12.7311, radius=7.9186, center=<5.4318 -3.3237 -1.8827>, rotation=<rot 0 104.5448 -71.2773>}]
 hand_r [{shapeType=Capsule, length=12.7259, radius=7.9537, center=<-5.4372 3.3492 1.8827>, rotation=<rot 0 104.727 -71.1856>}]
 head [{shapeType=Capsule, length=23.6328, radius=20.2546, center=<7.326 -0.3472 -0.0017>, rotation=<rot 0 -173.4994 -89.9833>}]
 lowerarm_l [{shapeType=Capsule, length=16.0032, radius=12.7027, center=<10.4936 0.487 -2.5917>, rotation=<rot -9.4946 -90 -81.4025>}]
 lowerarm_r [{shapeType=Capsule, length=16.0032, radius=12.7027, center=<-10.4936 -0.487 2.5917>, rotation=<rot -9.4947 -90 -81.4026>}]
 pelvis [{shapeType=Capsule, length=37.759, radius=36.8489, center=<4.6099 -0.3536 -0.1144>, rotation=<rot 7.7792 97.8439 -90>}]
 pinky_metacarpal_l [{shapeType=Capsule, length=11.2894, radius=6.3559, center=<3.5086 -1.7929 1.8005>, rotation=<rot 0 -112.9136 -64.9137>}]
 pinky_metacarpal_r [{shapeType=Capsule, length=11.2893, radius=6.356, center=<-3.5087 1.793 -1.8005>, rotation=<rot 0 -112.9135 -64.9137>}]
 spine_04 [{shapeType=Capsule, length=23.4916, radius=13.2467, center=<8.9795 -17.2934 0.0365>, rotation=<rot -0.1241 -90 -0.3467>}]
 spine_05 [{shapeType=Capsule, length=26.4475, radius=12.0254, center=<-8.4398 18.5212 0.0255>, rotation=<rot 0.3791 -90 -0.5568>}]
 thigh_twist_01_l [{shapeType=Capsule, length=20.4915, radius=10.6719, center=<0.6448 -1.8158 -7.9754>, rotation=<rot 0 2.4451 -78.7736>}]
 thigh_twist_01_r [{shapeType=Capsule, length=20.3197, radius=10.6082, center=<-0.3018 2.097 8.085>, rotation=<rot 0 2.4965 -78.0927>}]
 thigh_twist_02_l [{shapeType=Capsule, length=17.7038, radius=15.0597, center=<-1.2934 0.0302 1.2267>, rotation=<rot 0 172.8811 -77.9248>}]
 thigh_twist_02_r [{shapeType=Capsule, length=17.7499, radius=15.2141, center=<0.8295 -0.1343 -1.9278>, rotation=<rot 0 170.9244 -63.7046>}]
 upperarm_l [{shapeType=Capsule, length=15.1331, radius=11.3847, center=<20.6567 0.2134 -0.7329>, rotation=<rot 0 10.1376 -81.4768>}]
 upperarm_r [{shapeType=Capsule, length=15.1523, radius=11.4535, center=<-20.6524 -0.2155 0.7715>, rotation=<rot 0 10.2997 -80.8571>}]
 upperarm_twist_01_l [{shapeType=Capsule, length=18.7799, radius=9.4928, center=<-2.1729 -2.5974 0.6096>, rotation=<rot 0 -166.4139 -84.2728>}]
 upperarm_twist_01_r [{shapeType=Capsule, length=18.7801, radius=9.4929, center=<2.173 2.5975 -0.6096>, rotation=<rot 0 -166.414 -84.2727>}]
}
constraints [
 {bone1Name=spine_04, bone2Name=pelvis, swing1LimitDegrees=45, swing2LimitDegrees=45, twistLimitDegrees=45}
 {bone1Name=spine_05, bone2Name=spine_04, swing1LimitDegrees=45, swing2LimitDegrees=45, twistLimitDegrees=45}
 {bone1Name=head, bone2Name=spine_05, swing1LimitDegrees=45, swing2LimitDegrees=45, twistLimitDegrees=45}
 {bone1Name=clavicle_l, bone2Name=spine_05, swing1LimitDegrees=45, swing2LimitDegrees=45, twistLimitDegrees=45}
 {bone1Name=upperarm_l, bone2Name=clavicle_l, swing1LimitDegrees=45, swing2LimitDegrees=45, twistLimitDegrees=45}
 {bone1Name=upperarm_twist_01_l, bone2Name=upperarm_l, swing1LimitDegrees=45, swing2LimitDegrees=45, twistLimitDegrees=45}
 {bone1Name=lowerarm_l, bone2Name=upperarm_l, swing1LimitDegrees=45, swing2LimitDegrees=45, twistLimitDegrees=45}
 {bone1Name=hand_l, bone2Name=lowerarm_l, swing1LimitDegrees=45, swing2LimitDegrees=45, twistLimitDegrees=45}
 {bone1Name=pinky_metacarpal_l, bone2Name=hand_l, swing1LimitDegrees=45, swing2LimitDegrees=45, twistLimitDegrees=45}
 {bone1Name=clavicle_r, bone2Name=spine_05, swing1LimitDegrees=45, swing2LimitDegrees=45, twistLimitDegrees=45}
 {bone1Name=upperarm_r, bone2Name=clavicle_r, swing1LimitDegrees=45, swing2LimitDegrees=45, twistLimitDegrees=45}
 {bone1Name=upperarm_twist_01_r, bone2Name=upperarm_r, swing1LimitDegrees=45, swing2LimitDegrees=45, twistLimitDegrees=45}
 {bone1Name=lowerarm_r, bone2Name=upperarm_r, swing1LimitDegrees=45, swing2LimitDegrees=45, twistLimitDegrees=45}
 {bone1Name=hand_r, bone2Name=lowerarm_r, swing1LimitDegrees=45, swing2LimitDegrees=45, twistLimitDegrees=45}
 {bone1Name=pinky_metacarpal_r, bone2Name=hand_r, swing1LimitDegrees=45, swing2LimitDegrees=45, twistLimitDegrees=45}
 {bone1Name=calf_l, bone2Name=pelvis, swing1LimitDegrees=45, swing2LimitDegrees=45, twistLimitDegrees=45}
 {bone1Name=foot_l, bone2Name=calf_l, swing1LimitDegrees=45, swing2LimitDegrees=45, twistLimitDegrees=45}
 {bone1Name=calf_twist_02_l, bone2Name=calf_l, swing1LimitDegrees=45, swing2LimitDegrees=45, twistLimitDegrees=45}
 {bone1Name=thigh_twist_01_l, bone2Name=pelvis, swing1LimitDegrees=45, swing2LimitDegrees=45, twistLimitDegrees=45}
 {bone1Name=thigh_twist_02_l, bone2Name=pelvis, swing1LimitDegrees=45, swing2LimitDegrees=45, twistLimitDegrees=45}
 {bone1Name=calf_r, bone2Name=pelvis, swing1LimitDegrees=45, swing2LimitDegrees=45, twistLimitDegrees=45}
 {bone1Name=foot_r, bone2Name=calf_r, swing1LimitDegrees=45, swing2LimitDegrees=45, twistLimitDegrees=45}
 {bone1Name=calf_twist_02_r, bone2Name=calf_r, swing1LimitDegrees=45, swing2LimitDegrees=45, twistLimitDegrees=45}
 {bone1Name=thigh_twist_01_r, bone2Name=pelvis, swing1LimitDegrees=45, swing2LimitDegrees=45, twistLimitDegrees=45}
 {bone1Name=thigh_twist_02_r, bone2Name=pelvis, swing1LimitDegrees=45, swing2LimitDegrees=45, twistLimitDegrees=45}
]
